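; Passive quest ability for the Season 23 Battle Pass that tracks the player picking augments.
; On activation it runs the parent logic, then listens on the owning player's pawn for the Event.Muster.AugAdded gameplay event.
; A received event sends a complex custom stat event with a count of one to the player's quest manager, passing the Additional Source Tags.
; The ability calls EndAbility when passive activation fails or its end event fires.
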(event Ability|EventOnEndAbility
  (params (bWasCancelled :type Boolean))
  (Ability|EndAbility))

(event Ability|EventFailedToActivatePassiveAbility
  (params (ActorInfo :type Gameplay%20Ability%20Actor%20Info%20Structure))
  (Ability|EndAbility))

(event Ability|EventActivateAbility
  (|CallParentFunction)
  (Ability|Async|WaitGameplayEventToActor (Variables|GAVars|Data|GetOwningPlayerPawn) "(TagName=\"Event.Muster.AugAdded\")"
    (:EventReceived
      (Quest|SendComplexCustomStatEvent (Variables|GAVars|Data|GetOwningPlayerQuestManager) 0 (Variables|Default|GetAdditional%20Source%20Tags) "(GameplayTags=)" 1))))
